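# EQS query for the Medic NPC specialist that picks a position close around its selected squad member.
# An OnCircle generator places a small ring of candidate points centered on the selected squad member context, with projection tracing turned off.
# A FortQueryTest_IsObstructed test uses the same context to score the points by obstruction; it only scores, so it filters none out.
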EnvQuery /NPCSpecialists/NPCs/Medic/EQS/EQS_FindPointAroundTarget
options [{
  generator {
   name=EnvQueryGenerator_OnCircle_0
   class=EnvQueryGenerator_OnCircle
   properties {
    circleCenter=&/EnvQueryContext_SP_Medic_SelectedSquadMember.EnvQueryContext_SP_Medic_SelectedSquadMember_C
    circleRadius=150
    numberOfPoints=4
    projectionData {extentX=0, navigationFilter=null, postProjectionVerticalOffset=0, projectDown=1024, traceMode=None}
    traceData {navigationFilter=null}
   }
  }
  tests [FortQueryTest_IsObstructed_0 {
    class=FortQueryTest_IsObstructed
    properties {
     context=&EnvQueryContext_SP_Medic_SelectedSquadMember_C
     testPurpose=Score
    }
   }]
 }]
