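# Behavior tree for the Medic NPC specialist that moves to the ActorToHeal and throws a Chug Splash through the Heal gameplay ability.
# A root service applies a LeashPaused tag while the tree runs.
# If the target is in a vehicle, the bot approaches and throws directly, with the tag check able to abort the lower-priority branch.
# Otherwise it approaches, picks a nearby point with EQS, moves there and throws.
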BehaviorTree /NPCSpecialists/NPCs/Medic/BT/BT_SP_Medic_Heal
blackboard=&/BB_SP_Medic
root {
 name=Selector
 class=BTComposite_Selector
 children [
  Sequence {
   class=BTComposite_Sequence
   children [
    "Bot Move To (Urgent)" {
     class=FortAthenaBTTask_BotMoveTo
     decorators []
     properties {
      movementResultKeyName=None
      nodeName="Bot Move To (Urgent)"
      acceptableRadius {defaultValue=300}
      allowRandomWobble {defaultValue=true}
      bAllowPartialPath {defaultValue=false}
      bAllowStrafe {defaultValue=true}
      blackboardKey {selectedKeyName=ActorToHeal}
      observedBlackboardValueTolerance {defaultValue=4.75}
     }
    }
    Throw Chug Splash at Player {
     class=FortBTTask_ExecuteGameplayAbility
     decorators []
     properties {nodeName=Throw Chug Splash at Player, gameplayAbilityTagContainer {defaultValue {gameplayTags [{tagName=Ability.Specialist.Medic.Heal}]}}}
    }
   ]
   decorators [ActorToHeal is in Vehicle? {
     class=BTDecorator_GameplayTagQuery
     properties {
      flowAbortMode=LowerPriority
      nodeName=ActorToHeal is in Vehicle?
      actorForGameplayTagQuery {selectedKeyName=ActorToHeal}
      gameplayTagQuery {autoDescription=" ANY( Gameplay.Status.InVehicle )", queryTokenStream [0,1,1,1,0], tagDictionary [{tagName=Gameplay.Status.InVehicle}]}
     }
    }]
   properties {children [
     {childTask=&self:FortAthenaBTTask_BotMoveTo_0}
     {childTask=&self:FortBTTask_ExecuteGameplayAbility_0}
    ]}
  }
  Sequence {
   class=BTComposite_Sequence
   children [
    "Bot Move To (Urgent)" {
     class=FortAthenaBTTask_BotMoveTo
     decorators []
     properties {
      movementResultKeyName=None
      nodeName="Bot Move To (Urgent)"
      acceptableRadius {defaultValue=300}
      allowRandomWobble {defaultValue=true}
      bAllowPartialPath {defaultValue=false}
      bAllowStrafe {defaultValue=true}
      blackboardKey {selectedKeyName=ActorToHeal}
      observedBlackboardValueTolerance {defaultValue=4.75}
     }
    }
    Find Loc Around SquadMember {
     class=BTTask_RunEQSQuery
     properties {
      bUpdateBBOnFail=true
      nodeName=Find Loc Around SquadMember
      blackboardKey {selectedKeyName=LocationToHeal}
      eQSRequest {queryTemplate=/NPCSpecialists/NPCs/Medic/EQS/EQS_FindPointAroundTarget, runMode=SingleResult}
     }
    }
    Bot Move To {
     class=FortAthenaBTTask_BotMoveTo
     decorators []
     properties {
      movementResultKeyName=None
      nodeName=Bot Move To
      acceptableRadius {defaultValue=10}
      allowRandomWobble {defaultValue=true}
      bAllowPartialPath {defaultValue=false}
      bAllowStrafe {defaultValue=true}
      blackboardKey {selectedKeyName=LocationToHeal}
      observedBlackboardValueTolerance {defaultValue=4.75}
     }
    }
    Throw Chug Splash at Player {
     class=FortBTTask_ExecuteGameplayAbility
     decorators []
     properties {nodeName=Throw Chug Splash at Player, gameplayAbilityTagContainer {defaultValue {gameplayTags [{tagName=Ability.Specialist.Medic.Heal}]}}}
    }
   ]
   decorators []
   properties {children [
     {childTask=&self:FortAthenaBTTask_BotMoveTo_15}
     {childTask=&self:BTTask_RunEQSQuery_0}
     {childTask=&self:FortAthenaBTTask_BotMoveTo_13}
     {childTask=&self:FortBTTask_ExecuteGameplayAbility_2}
    ]}
  }
 ]
 decorators []
 properties {
  children [
   {
    childComposite=&self:BTComposite_Sequence_2
    childTask=null
    decorators [&self:BTDecorator_GameplayTagQuery_7]
   }
   {childComposite=&self:BTComposite_Sequence_7, childTask=null, decorators []}
  ]
  services [&self:FortAthenaBTService_ApplyGameplayTags_1]
 }
 services [Pause Leash {
   class=FortAthenaBTService_ApplyGameplayTags
   properties {interval=3600, nodeName=Pause Leash, gameplayTagsToApply {defaultValue {gameplayTags [{tagName=Athena.Gameplay.Status.LeashPaused}]}}}
  }]
}
